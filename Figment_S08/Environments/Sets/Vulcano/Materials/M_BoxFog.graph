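# M_BoxFog is a translucent, unlit fog material in the Vulcano environment set, named for box-shaped fog.
# All of its shading comes from a single MF_BoxHeightFog function call, using material attributes.
# The material itself adds no other nodes, so the fog logic lives in that function.
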Material /Figment_S08/Environments/Sets/Vulcano/Materials/M_BoxFog
graph {
 properties {bUseMaterialAttributes=true, blendMode=BLEND_Translucent, pixelDepthOffsetMode=PDOM_Legacy, shadingModel=MSM_Unlit, usedShadingModels=Unlit}
 nodes [MaterialFunctionCall!(In) {materialFunction=&/MF_BoxHeightFog}]
}
