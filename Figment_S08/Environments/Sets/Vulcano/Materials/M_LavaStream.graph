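# Translucent unlit lava stream material for the Vulcano environment set, usable on spline meshes.
# A panning mask texture's G channel blends dark to time-of-day stream colors, with branches for eye-adaptation compensation and brighter HLOD proxy baking.
# Its R channel, scaled and multiplied by vertex-color blue, forms an opacity term.
# A separate desktop/mobile emissive multiplier tints StreamColor, but feeds only the Morning channel into all four time-of-day inputs.
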Material /Figment_S08/Environments/Sets/Vulcano/Materials/M_LavaStream
graph {
 properties {bUsedWithSplineMeshes=true, blendMode=BLEND_Translucent, pixelDepthOffsetMode=PDOM_Legacy, shadingModel=MSM_Unlit, usedShadingModels=Unlit}
 nodes [
  Constant3Vector()
  n1 = ScalarParameter() {parameterName=PanSpeedStream2_X}
  n2 = ScalarParameter() {defaultValue=-0.05, parameterName=PanSpeedStream2_Y}
  n3 = TextureSample(UVs=Panner(Coordinate=TextureCoordinate(), Speed=AppendVector(A=@n1, B=@n2))) {samplerType=SAMPLERTYPE_Masks, texture=/Game/Environments/Sets/Vulcano/Textures/LavaStream_M}
  n4 = LinearInterpolate(
   A=VectorParameter() {parameterName=StreamColor_Dark, defaultValue=<rgba 0 0 0 1>}.RGB
   B=MaterialFunctionCall(
    Color_Morning=VectorParameter() {parameterName=StreamColor_Morning, defaultValue=<rgba 5 1.1178 0.4948 1>}.RGB
    Color_Day=VectorParameter() {parameterName=StreamColor_Day, defaultValue=<rgba 5 1.1178 0.4948 1>}.RGB
    Color_Evening=VectorParameter() {parameterName=StreamColor_Evening, defaultValue=<rgba 5 1.1178 0.4948 1>}.RGB
    Color_Night=VectorParameter() {parameterName=StreamColor_Night, defaultValue=<rgba 5 1.1178 0.4948 1>}.RGB
   ) {materialFunction=&Color_Time_of_Day}.Result
   Alpha=@n3.G
  )
  Divide(A=@n4, B=EyeAdaptation().EyeAdaptation)
  MaterialProxyReplace(Realtime=@n4, MaterialProxy=Multiply(A=@n4, B=ScalarParameter() {defaultValue=5, parameterName=HlodBirghtness}))
  n5 = ScalarParameter() {defaultValue=1, parameterName=Opcacity}
  Multiply(A=Clamp(None=Multiply(A=@n3.R, B=@n5)), B=VertexColor().B)
  n6 = VectorParameter() {parameterName=TOD_EmissiveMult, channelNames {a=Night, b=Evening, g=Day, r=Morning}, defaultValue=<rgba 1 1 1 1>}
  n7 = MaterialFunctionCall(Value_Morning=@n6.Morning, Value_Day=@n6.Morning, Value_Evening=@n6.Morning, Value_Night=@n6.Morning) {materialFunction=&Scalar_Time_of_Day}
  n8 = VectorParameter() {parameterName=TOD_EmissiveMult_Mobile, channelNames {a=Night, b=Evening, g=Day, r=Morning}, defaultValue=<rgba 1 1 1 1>}
  n9 = MaterialFunctionCall(Value_Morning=@n8.Morning, Value_Day=@n8.Morning, Value_Evening=@n8.Morning, Value_Night=@n8.Morning) {materialFunction=&Scalar_Time_of_Day}
  Multiply(
   A=NamedRerouteDeclaration(None=ShadingPathSwitch(Default=@n7.Result, Mobile=@n9.Result)) {name=TODMult, nodeColor=<rgba 0 1 0.6719 1>}
   B=VectorParameter() {parameterName=StreamColor, defaultValue=<rgba 5 1.1178 0.4948 1>}.RGB
  )
  VertexColor()
 ]
}
