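# Post-process material, applied after DOF, for a radial streak burst centred on a world-space location, likely a volcano event.
# The projected location drives radial UVs for a streak mask that distorts and tints the scene, timed by the Panner parameter.
# View alignment adds desaturation, red and fiery cores, while offset scene-depth samples form dark streaked outlines.
# A screen-centre sphere mask keeps the original image, and Blend controls overall strength.
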Material /Figment_S08/Environments/Sets/Vulcano/Materials/PPM_Streaks
graph {
 properties {blendableLocation=BL_SceneColorAfterDOF, materialDomain=MD_PostProcess, pixelDepthOffsetMode=PDOM_Legacy, usedShadingModels=Default Lit}
 nodes [
  n1 = Multiply(A=Time(), B=ScalarParameter() {defaultValue=0.5, parameterName=TimeSpeed})
  Clamp(None=Multiply(A=Frac(None=@n1), B=ScalarParameter() {defaultValue=2, parameterName=WaitTime}))
  n2 = ComponentMask(None=SceneTexture() {sceneTextureId=PPI_PostProcessInput0}.Color) {b=true, g=true, r=true}
  n3 = ScalarParameter() {defaultValue=-0.01, parameterName=Distortion}
  n4 = Multiply(A=@n3) {constB=-1}
  n5 = VectorParameter() {parameterName=Location, defaultValue=<rgba 25000 -65000 -1500 1>}
  n6 = Subtract(
   A=ScreenPosition().ViewportUV
   B=MaterialFunctionCall(World Position=@n5.RGB) {materialFunction=&TransformToClipSpace}.Buffer UVs for Scene Textures
  )
  n7 = MaterialFunctionCall(Vector or UVs=@n6) {materialFunction=&VectorToRadialValue}
  n8 = AppendVector(A=ScalarParameter() {defaultValue=1, parameterName=TilingX2}, B=ScalarParameter() {defaultValue=0.219, parameterName=TilingY2})
  n9 = Frac(None=ScalarParameter() {parameterName=Panner})
  n10 = AppendVector(
   A=ScalarParameter() {parameterName=SpeedX}
   B=LinearInterpolate(A=ScalarParameter() {defaultValue=2, parameterName=SpeedY}, B=Constant(), Alpha=@n9)
  )
  n11 = TextureCoordinate()
  n12 = TextureSample!(
   UVs=Add(A=Multiply(A=@n7.Radial Coordinates, B=@n8), B=@n10)
   Tex
   "DDX(UVs)"=DDX(Value=@n11)
   "DDY(UVs)"=DDY(Value=@n11)
   Apply View MipBias
  ) {mipValueMode=TMVM_Derivative, samplerType=SAMPLERTYPE_Masks, texture=/Game/Environments/Sets/Vulcano/Textures/T_LineStreaks_M}
  n13 = Normalize(VectorInput=@n6)
  n14 = ScalarParameter() {defaultValue=0.2838, parameterName=DistortionSpeedBias}
  n15 = OneMinus(None=@n9)
  n16 = ScalarParameter() {defaultValue=0.5419, parameterName=DistortionSpeedBias2}
  n17 = Clamp(None=Multiply(A=Power(Base=@n9, Exp=@n14), B=Power(Base=@n15, Exp=@n16)))
  n18 = Multiply(A=Multiply(A=LinearInterpolate(A=@n3, B=@n4, Alpha=@n12.R), B=@n13), B=@n17)
  n19 = ScreenPosition()
  n20 = CameraPositionWS()
  n21 = DotProduct(A=Normalize(VectorInput=Subtract(A=@n5.RGB, B=@n20)), B=CameraVectorWS())
  n22 = Clamp(None=Multiply(A=@n21) {constB=-1})
  n23 = Multiply(A=@n22, B=LinearInterpolate(A=Constant(), B=ScalarParameter() {defaultValue=-0.051, parameterName=DistortionUniform}, Alpha=@n17))
  n24 = Add(A=@n18, B=Add(A=@n19.ViewportUV, B=Multiply(A=@n23, B=@n13)))
  n25 = SceneTexture(UVs=@n24) {sceneTextureId=PPI_PostProcessInput0}
  n26 = ComponentMask(None=SceneTexture(UVs=Add(A=@n19.ViewportUV, B=@n18)) {sceneTextureId=PPI_PostProcessInput0}.Color) {b=true, g=true, r=true}
  n27 = AppendVector(A=ScalarParameter() {defaultValue=3, parameterName=TilingX}, B=ScalarParameter() {defaultValue=1.6537, parameterName=TilingY})
  n28 = TextureCoordinate()
  n29 = TextureSample(UVs=Add(A=Multiply(A=@n7.Radial Coordinates, B=@n27), B=@n10), "DDX(UVs)"=DDX(Value=@n28), "DDY(UVs)"=DDY(Value=@n28)) {mipValueMode=TMVM_Derivative, samplerType=SAMPLERTYPE_Masks, texture=&T_LineStreaks_M}
  n30 = LinearInterpolate(
   A=VectorParameter() {parameterName=Color1, defaultValue=<rgba 1 1 1 1>}.RGB
   B=VectorParameter() {parameterName=Color2, defaultValue=<rgba 0 0 0 1>}.RGB
   Alpha=@n29.R
  )
  n31 = Add(
   A=Distance(A=ScreenPosition().ViewportUV, B=Constant2Vector() {g=0.5, r=0.5})
   B=ScalarParameter() {defaultValue=-0.2514, parameterName=FallofVignetteOffset}
  )
  n32 = ScalarParameter() {defaultValue=0.3967, parameterName=FallofVignette}
  n33 = Multiply(A=Clamp(None=Divide(A=@n31, B=@n32)), B=OneMinus(None=Clamp(None=@n21)))
  n34 = Sine(None=Divide(A=@n9))
  n35 = Saturate(None=Multiply(A=@n34, B=ScalarParameter() {defaultValue=1.6684, parameterName=SpeedupCore}))
  n36 = Desaturation(
   None=LinearInterpolate(A=ComponentMask(None=@n25.Color) {b=true, g=true, r=true}, B=Multiply(A=@n26, B=@n30), Alpha=Multiply(A=@n33, B=@n35))
   Fraction=@n35
  )
  n37 = Power(Base=@n22, Exp=ScalarParameter() {defaultValue=16.1368, parameterName=RedCoreFade})
  n38 = LinearInterpolate(A=@n36, B=VectorParameter() {parameterName=Red, defaultValue=<rgba 1 0 0 1>}.RGB, Alpha=Multiply(A=@n34, B=Clamp(None=@n37)))
  n39 = VectorParameter() {parameterName=DarkStreaks, defaultValue=<rgba 0.1 0 0.0367 1>}
  n40 = Multiply(A=@n34, B=@n29.G)
  n41 = VectorParameter() {parameterName=Color3, defaultValue=<rgba 0.1406 0.0795 0.0332 1>}
  n42 = VectorParameter() {parameterName=Color4, defaultValue=<rgba 5 1.9151 0.1835 1>}
  n43 = ScalarParameter() {defaultValue=50, parameterName=InnerCoreFade}
  n44 = OneMinus(None=Clamp(None=Power(Base=@n22, Exp=@n43)))
  n45 = LinearInterpolate(A=ScalarParameter() {defaultValue=100, parameterName=Falloff1}, B=ScalarParameter() {defaultValue=80, parameterName=Falloff1x}, Alpha=@n35)
  n46 = Multiply(A=@n35, B=Clamp(None=Power(Base=@n22, Exp=@n45)))
  n47 = LinearInterpolate(
   A=LinearInterpolate(
    A=ScalarParameter() {defaultValue=-0.005, parameterName=SceneDepthOffset2}
    B=ScalarParameter() {defaultValue=-0.029, parameterName=SceneDepthOffset}
    Alpha=@n34
   )
   B=Constant()
   Alpha=@n29.R
  )
  n48 = SceneTexture(UVs=Add(A=@n24, B=Multiply(A=@n13, B=@n47))) {sceneTextureId=PPI_SceneDepth}
  n49 = ComponentMask(None=SceneTexture(UVs=@n24) {sceneTextureId=PPI_SceneDepth}.Color) {r=true}
  n50 = Divide(A=Subtract(A=ComponentMask(None=@n48.Color) {r=true}, B=@n49), B=ScalarParameter() {defaultValue=8000, parameterName=Outline})
  n51 = Clamp(None=Multiply(A=@n34, B=ScalarParameter() {defaultValue=3, parameterName=FadeInOutlineSpeed}))
  n52 = Subtract(A=@n49, B=Distance(A=@n5.RGB, B=CameraPositionWS()))
  n53 = ScalarParameter() {defaultValue=763496.0625, parameterName=FadeOutOutline}
  n54 = OneMinus(None=Clamp(None=Divide(A=@n52, B=@n53)))
  n55 = LinearInterpolate(
   A=VectorParameter() {parameterName=Shadows, defaultValue=<rgba 0.0469 0 0.0176 1>}.RGB
   B=LinearInterpolate(
    A=LinearInterpolate(A=@n38, B=@n39.RGB, Alpha=@n40)
    B=LinearInterpolate(A=@n41.RGB, B=@n42.RGB, Alpha=@n12.R)
    Alpha=Multiply(A=@n44, B=@n46)
   )
   Alpha=LinearInterpolate(A=Constant() {r=1}, B=Clamp(None=@n50), Alpha=Multiply(A=@n51, B=@n54))
  )
  n56 = SceneDepth()
  n57 = ScalarParameter() {defaultValue=25000, parameterName=OffsetDepthFadeDistance}
  n58 = CameraPositionWS()
  n59 = Normalize(VectorInput=Subtract(A=WorldPosition().XYZ, B=@n58))
  n60 = Normalize(VectorInput=Subtract(A=@n5.RGB, B=@n58))
  n61 = Multiply(A=Saturate(None=Divide(A=@n56.R, B=@n57)), B=Saturate(None=DotProduct(A=@n59, B=@n60)))
  n62 = ScreenPosition()
  n63 = Constant() {r=0.5}
  n64 = ScalarParameter() {defaultValue=0.5, parameterName=Radius}
  n65 = ScalarParameter() {parameterName=Hardness}
  LinearInterpolate(
   A=@n2
   B=LinearInterpolate(A=LinearInterpolate(A=@n36, B=@n55, Alpha=@n61), B=@n2, Alpha=SphereMask(A=@n62.ViewportUV, B=@n63, Radius=@n64, Hardness=@n65))
   Alpha=ScalarParameter() {parameterName=Blend}
  )
 ]
}
